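# Conversation database for the BentBaton NPC's "Ask about..." dialogue, providing optional flavor chat about the character.
# The Conversation.Entry.BentBaton.Dialogue entry point leads to a "How can I help?" hub offering three topics and a Back choice.
# Cheerleading, The Beast and Her Puppuccino's Order each play as a linear chain of Speech nodes.
# All three chains converge on one shared ReturnToStart node, while the hub's BACK uses a separate Back node.
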ConversationDatabase /NPCLibrary/NPCs/BentBaton/NPCLibrary_Conversation_BentBaton_Dialogue
entry_points [{tag=Conversation.Entry.BentBaton.Dialogue, destinations [#1]}]
nodes [
 ConversationEntryPointNode_0 {
  id=#1
  class=ConversationEntryPointNode
  outputs [#2]
  properties {entryTag {tagName=Conversation.Entry.BentBaton.Dialogue}}
 }
 Speech_C_0 {
  id=#2
  class=Speech_C
  outputs [#3, #4, #5, #6]
  properties {nodeName="", softTaskIcon=&T_UI_NPC_question, generalConfig {defaultMessage=How can I help?}}
  sub_nodes [ConversationChoiceNode_5 {class=ConversationChoiceNode, properties {defaultChoiceDisplayText=Ask about...}}]
 }
 Speech_C_6 {
  id=#6
  class=Speech_C
  outputs [#7]
  properties {
   nodeName=""
   softTaskIcon=&T_UI_NPC_question
   generalConfig {defaultMessage="It's a medium iced latte with a protein booster, a cinnamon swizzle stick..."}
  }
  sub_nodes [ConversationChoiceNode_15 {class=ConversationChoiceNode, properties {defaultChoiceDisplayText=Her Puppuccino's Order}}]
 }
 Speech_C_8 {
  id=#7
  class=Speech_C
  outputs [#8]
  properties {nodeName="", softTaskIcon=&T_UI_NPC_question, generalConfig {defaultMessage="...and an egg. "}}
 }
 Speech_C_9 {
  id=#8
  class=Speech_C
  outputs [#9]
  properties {
   nodeName=""
   softTaskIcon=&T_UI_NPC_question
   generalConfig {defaultMessage="Yes, I stir a poached egg into my latte! It's healthy, delicious, and time-efficient!!"}
  }
 }
 Speech_C_10 {
  id=#9
  class=Speech_C
  outputs [#10]
  properties {
   nodeName=""
   softTaskIcon=&T_UI_NPC_question
   generalConfig {defaultMessage="Did Ziggy tell you about the egg thing? He never lets me live it down!"}
  }
 }
 ReturnToStart_C_23 {
  id=#10
  class=ReturnToStart_C
  properties {softTaskIcon=&T_UI_NPC_Back}
  sub_nodes [ConversationChoiceNode_32 {class=ConversationChoiceNode, properties {defaultChoiceDisplayText=BACK}}]
 }
 Speech_C_2 {
  id=#5
  class=Speech_C
  outputs [#11]
  properties {
   nodeName=""
   softTaskIcon=&T_UI_NPC_question
   generalConfig {defaultMessage="There's nothing like bein' the beast!"}
  }
  sub_nodes [ConversationChoiceNode_1 {class=ConversationChoiceNode, properties {defaultChoiceDisplayText=The Beast}}]
 }
 Speech_C_3 {
  id=#11
  class=Speech_C
  outputs [#12]
  properties {
   nodeName=""
   softTaskIcon=&T_UI_NPC_question
   generalConfig {defaultMessage="It makes me feel powerful, sure. But also gives me this deep sense of... peace?"}
  }
 }
 Speech_C_4 {
  id=#12
  class=Speech_C
  outputs [#10]
  properties {
   nodeName=""
   softTaskIcon=&T_UI_NPC_question
   generalConfig {defaultMessage="No anxiety, no doubt. Just the raging, perfect storm."}
  }
 }
 Back_C_1 {
  id=#4
  class=Back_C
  sub_nodes [ConversationChoiceNode_2 {class=ConversationChoiceNode, properties {defaultChoiceDisplayText=BACK}}]
 }
 Speech_C_7 {
  id=#3
  class=Speech_C
  outputs [#13]
  properties {
   nodeName=""
   softTaskIcon=&T_UI_NPC_question
   generalConfig {defaultMessage=It's the most exhausting and exhilarating sport I've ever done.}
  }
  sub_nodes [ConversationChoiceNode_0 {class=ConversationChoiceNode, properties {defaultChoiceDisplayText=Cheerleading}}]
 }
 Speech_C_15 {
  id=#13
  class=Speech_C
  outputs [#14]
  properties {
   nodeName=""
   softTaskIcon=&T_UI_NPC_question
   generalConfig {defaultMessage="Most athletes get to swap on and off the field during the game. Not cheerleaders!"}
  }
 }
 Speech_C_14 {
  id=#14
  class=Speech_C
  outputs [#15]
  properties {
   nodeName=""
   softTaskIcon=&T_UI_NPC_question
   generalConfig {defaultMessage="We're out there the whole time, flipping, flying, and cheering at the top of our lungs."}
  }
 }
 Speech_C_1 {
  id=#15
  class=Speech_C
  outputs [#10]
  properties {
   nodeName=""
   softTaskIcon=&T_UI_NPC_question
   generalConfig {defaultMessage="It's hard work, but pepping up the crowd... nothing beats it!"}
  }
 }
]
